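# Physics asset for the KpopFashion male medium player body, covering a dangling microphone accessory and the hip area it hangs near.
# Capsules form a mic, string and string-base chain, with swing and twist limits on its two joints and very tight twist at the base.
# Pelvis and thigh capsules approximate the body, likely as collision for the mic.
# Constraints from the pelvis to both thighs and the string base leave swing and twist free.
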PhysicsAsset /BRCosmetics/Characters/Player/Male/Medium/Bodies/M_MED_KpopFashion/Meshes/M_MED_Kpop_Fashion_Dyn
bodies {
 dyn_mic [{shapeType=Capsule, length=21.8042, radius=1.8764, center=<14.8544 -0.5934 0.8913>, rotation=<rot 0 87.8775 -86.27>}]
 dyn_mic_string [{shapeType=Capsule, length=2.7166, radius=1.7287, center=<1.5009 0 0>, rotation=<rot 90 0 0>}]
 dyn_mic_string_base [{shapeType=Capsule, length=0.505, radius=0.505, shapeName=dyn_mic_string_base_capsule, rotation=<rot 0 -90 0>}]
 pelvis [{shapeType=Capsule, length=9.875, radius=10.5, center=<-8.0395 0.9256 0>}]
 thigh_l [{shapeType=Capsule, length=23.25, radius=7.875, center=<-18.8061 0.7075 0>, rotation=<rot 90 -90 -84.2603>}]
 thigh_r [{shapeType=Capsule, length=23.25, radius=7.875, center=<18.8061 -0.7075 0>, rotation=<rot 90 -90 -84.2603>}]
}
constraints [
 {bone1Name=dyn_mic, bone2Name=dyn_mic_string, swing1LimitDegrees=60, swing2LimitDegrees=60, twistLimitDegrees=20}
 {bone1Name=dyn_mic_string, bone2Name=dyn_mic_string_base, swing1LimitDegrees=60, swing2LimitDegrees=60, twistLimitDegrees=1}
 {
  bone1Name=thigh_r
  bone2Name=pelvis
  swing1LimitDegrees=45
  swing1Motion=Free
  swing2LimitDegrees=45
  swing2Motion=Free
  twistLimitDegrees=45
  twistMotion=Free
 }
 {
  bone1Name=thigh_l
  bone2Name=pelvis
  swing1LimitDegrees=45
  swing1Motion=Free
  swing2LimitDegrees=45
  swing2Motion=Free
  twistLimitDegrees=45
  twistMotion=Free
 }
 {
  bone1Name=dyn_mic_string_base
  bone2Name=pelvis
  swing1LimitDegrees=45
  swing1Motion=Free
  swing2LimitDegrees=45
  swing2Motion=Free
  twistLimitDegrees=45
  twistMotion=Free
 }
]
