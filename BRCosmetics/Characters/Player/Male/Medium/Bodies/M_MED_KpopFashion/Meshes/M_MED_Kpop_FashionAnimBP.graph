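# Post-process style AnimBP for the male KpopFashion body cosmetic, copying the parent mesh pose and adding secondary motion.
# bRigidBodyEnabled switches between RigidBody simulation relative to spine_02 and a cached "MobileDynamics" fallback using spring and look-at on the mic.
# Additive bone offsets then adjust both neck correctives and the mic string base.
# The string offset scales with LegUp_R, computed each update from right thigh and pelvis deltas from ref pose.
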AnimBlueprint /BRCosmetics/Characters/Player/Male/Medium/Bodies/M_MED_KpopFashion/Meshes/M_MED_Kpop_FashionAnimBP
target_skeleton=&/M_MED_Kpop_Fashion_Skeleton
graphs [
 AnimGraph {nodes [
   n1 = RigidBody(ComponentPose=Local To Component(LocalPose=Copy Pose From Mesh() {bUseAttachedParent=true})) {
    bClampLinearTranslationLimitToRefPose=true
    bOverrideWorldGravity=true
    overridePhysicsAsset=&/M_MED_Kpop_Fashion_Dyn
    simulationSpace=BaseBoneSpace
    alphaBoolBlend {blendOption=Linear}
    baseBoneRef {boneName=spine_02}
    componentAppliedLinearAccClamp=<2000 2000 2000>
    componentLinearAccScale=<0.5 0.2 0.2>
    componentLinearVelScale=<0.5 0.2 0.2>
    overrideWorldGravity {z=-980}
    simSpaceSettings {maxAngularAcceleration=10, maxAngularVelocity=3, maxLinearAcceleration=250, maxLinearVelocity=300, worldAlpha=1, externalLinearDragV {z=0.1}}
   }
   n2 = Component To Local(ComponentPose=@n1)
   n3 = Use cached pose 'MobileDynamics'()
   n4 = Get bRigidBodyEnabled()
   n5 = "Transform (Modify) Bone - Bone: dyn_mic_string_base"(
    ComponentPose=Local To Component(LocalPose=Blend Poses by bool(BlendPose_0=@n2, BlendPose_1=@n3, BlendTime_0=0, BlendTime_1=0, bActiveValue=@n4) {blendType=Linear})
    Alpha=Get LegUp_R() -> (LegUp_R:double)
    Translation="1.5,0,-0.5"
    Rotation="0,0,0"
    Scale="1,1,1"
   ) {
    rotationMode=BMM_Additive
    rotationSpace=BCS_BoneSpace
    translationMode=BMM_Additive
    translationSpace=BCS_BoneSpace
    alphaBoolBlend {blendOption=Linear}
    boneToModify {boneName=dyn_mic_string_base}
    translation {x=1.5, z=-0.5}
   }
   n6 = "Transform (Modify) Bone - Bone: neck_l_cor"(ComponentPose=@n5, Alpha=1, Translation="4,4,0", Rotation="0,0,0", Scale="1,1,1") {
    translationMode=BMM_Additive
    translationSpace=BCS_BoneSpace
    alphaBoolBlend {blendOption=Linear}
    boneToModify {boneName=neck_l_cor}
    translation=<4 4>
   }
   n7 = "Transform (Modify) Bone - Bone: neck_r_cor"(ComponentPose=@n6, Alpha=1, Translation="4,4,0", Rotation="0,0,0", Scale="1,1,1") {
    translationMode=BMM_Additive
    translationSpace=BCS_BoneSpace
    alphaBoolBlend {blendOption=Linear}
    boneToModify {boneName=neck_r_cor}
    translation=<4 4>
   }
   Output Pose(Result=Component To Local(ComponentPose=@n7)) {layerGroup=None}
   n8 = Spring controller - Bone: mic_aim_target!(
    ComponentPose:Component Space Pose Link Structure=Local To Component(LocalPose=Copy Pose From Mesh() {bUseAttachedParent=true})
    Alpha:float=1
   ) -> (Pose:Component Space Pose Link Structure) {
    is=title AnimGraphNode_SpringBone
    bLimitDisplacement=true
    bTranslateZ=false
    maxDisplacement=200
    springDamping=2
    springStiffness=20
    alphaBoolBlend {blendOption=Linear}
    springBone {boneName=mic_aim_target}
   }
   n9 = Look At - Bone: dyn_mic(ComponentPose=@n8, Alpha=1) {
    alphaBoolBlend {blendOption=Linear}
    boneToModify {boneName=dyn_mic}
    lookAtLocation {x=0}
    lookAtTarget {boneReference {boneName=mic_aim_target}}
    lookAt_Axis {axis=<1 0>}
    lookUp_Axis {axis {x=0, z=1}}
   }
   Save cached pose 'MobileDynamics'(Pose=Component To Local(ComponentPose=@n9)) {cacheName=MobileDynamics}
  ]}
 EventGraph {nodes [
   Event BlueprintUpdateAnimation()
   n1 = GetOwnerSkeletalMeshComponent()
   n2 = BreakRotator(InRot=BreakTransform(InTransform=GetDeltaTransformFromRefPose(self=@n1, BoneName=thigh_r, BaseName=pelvis)).Rotation)
   n3 = BreakTransform!(InTransform=GetDeltaTransformFromRefPose(self=@n1, BoneName=pelvis, BaseName=root)) -> (Location_X:double, Location_Y:double, Location_Z:double, Rotation, Scale)
   > Set LegUp_R(LegUp_R:double=float * float(
     A=MapRangeClamped(Value=@n2.Yaw, InRangeA=-40, InRangeB=-20, OutRangeA=1, OutRangeB=0)
     B=MapRangeClamped(Value=@n3, InRangeA=-35, InRangeB=-10, OutRangeA=1, OutRangeB=0)
    ) {is=title K2Node_PromotableOperator}) -> (Output_Get:double)
  ]}
]
variables [LegUp_R {type=number}]
